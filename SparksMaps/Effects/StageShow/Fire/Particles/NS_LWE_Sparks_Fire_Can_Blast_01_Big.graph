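# Niagara system for the large variant of a fire-can blast in the Sparks stage show.
# Main_Blast bursts four mesh flames, staggering size, height and material parameters by execution index, with curves driving rise, stretch and fade.
# Particle position is overwritten each update with random XY jitter for flicker, while Light001 adds an orange light that fades in and out.
# Four LWE emitters (base, glow, top and bottom end flames) list no modules or renderers here.
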NiagaraSystem /SparksMaps/Effects/StageShow/Fire/Particles/NS_LWE_Sparks_Fire_Can_Blast_01_Big
dependencies {
 usedDataInterfaces [&NiagaraDataInterfaceCurve]
 usedDynamicInputs [&MakeVector2D]
 usedModules [&SystemState]
}
emitters {
 LWE_Base {rendererClasses [], renderers []}
 LWE_EndFlames_Bottom {rendererClasses [], renderers []}
 LWE_EndFlames__Top {rendererClasses [], renderers []}
 LWE_Glow {rendererClasses [], renderers []}
 Light001 {
  emitterUpdateScript [
   &EmitterState(Life Cycle Mode=::NewEnumerator1, Loop Behavior=::NewEnumerator1, Loop Duration=1, UseLoopDelay=false)
   &SpawnBurst_Instantaneous(Use Spawn Probability=false, UseLoopCountLimit=false)
  ]
  particleSpawnScript [&InitializeParticle(Color=<rgba 50 9.1435 0 1>, Position Offset=<0 0 100>, UsePositionOffset=true, Write Material Random=true)]
  particleUpdateScript [
   &ParticleState()
   &ScaleColor(
    Scale RGB=&VectorFromFloat(Value=&FloatFromCurve(FloatCurve=DataInterface() {propertyValues {Curve {keys [
          {interpMode=RCIM_Cubic, tangentMode=RCTM_User, time=0, value=0}
          {interpMode=RCIM_Cubic, tangentMode=RCTM_User, time=0.22, value=1}
          {interpMode=RCIM_Cubic, tangentMode=RCTM_User, time=0.8, value=1}
          {interpMode=RCIM_Cubic, tangentMode=RCTM_User, time=1, value=0}
         ]}}}))
    ScaleA=true
    ScaleRGB=true
   )
  ]
  rendererClasses [&NiagaraLightRendererProperties]
  renderers [{rendererClass=&NiagaraLightRendererProperties, rendererIndex=0}]
 }
 Main_Blast {
  emitterUpdateScript [
   &EmitterState(Life Cycle Mode=::NewEnumerator1, Loop Behavior=::NewEnumerator1, Loop Duration=1, UseLoopDelay=false)
   &SpawnBurst_Instantaneous(Spawn Count=4, Use Spawn Probability=false, UseLoopCountLimit=false)
  ]
  particleSpawnScript [&InitializeParticle(
    Lifetime=1.25
    Mesh Scale=&MakeVector(
     X=1
     Y=&Add_Float(
      A=&RandomRangeFloat(Minimum=1, Maximum=1.25)
      B=&FloatFromCurve(
       FloatCurve=DataInterface() {propertyValues {Curve {keys [
           {arriveTangent=0.0556, interpMode=RCIM_Cubic, leaveTangent=0.0556, tangentMode=RCTM_User, time=0, value=0.1}
           {arriveTangent=1.445, interpMode=RCIM_Cubic, leaveTangent=1.445, tangentMode=RCTM_User, time=0.5, value=0.5}
           {interpMode=RCIM_Cubic, tangentMode=RCTM_User, time=1, value=1}
          ]}}}
       CurveIndex=&ReturnNormalizedExecIndex()
       Scale Curve=6
      )
     )
     Z=6
    )
    Mesh Scale Mode=::NewEnumerator4
    UsePositionOffset=false
    Write Material Random=true
   )]
  particleUpdateScript [
   &ParticleState()
   /SetVariables(Particles.Position=&AddVectorToPosition(
     Position=<0 0 0>
     Vector=&MakeVector(
      X=&RandomRangeFloat(Minimum=-2, Maximum=2)
      Y=&RandomRangeFloat(Minimum=-2, Maximum=2)
      Z=&FloatFromCurve(
       FloatCurve=DataInterface() {propertyValues {Curve {keys [
           {arriveTangent=3.4502, interpMode=RCIM_Cubic, leaveTangent=3.4502, tangentMode=RCTM_User, time=0, value=0}
           {arriveTangent=0.0241, interpMode=RCIM_Cubic, leaveTangent=0.0241, tangentMode=RCTM_User, time=0.7, value=1.125}
           {interpMode=RCIM_Cubic, tangentMode=RCTM_User, time=1, value=0.8}
          ]}}}
       Scale Curve=&Multiply_Float(A=300, B=&ReturnNormalizedExecIndex())
      )
     )
    )) {isSetParameters=true}
   &ScaleMeshSize(Scale Factor=&MakeVector(
     X=1
     Y=&FloatFromCurve(FloatCurve=DataInterface() {propertyValues {Curve {keys [
          {arriveTangent=-0.0123, interpMode=RCIM_Cubic, leaveTangent=-0.0123, tangentMode=RCTM_User, time=0, value=0.25}
          {arriveTangent=-0.0423, interpMode=RCIM_Cubic, leaveTangent=-0.0423, tangentMode=RCTM_User, time=0.2, value=1.25}
          {arriveTangent=0.0402, interpMode=RCIM_Cubic, leaveTangent=0.0402, tangentMode=RCTM_User, time=0.4, value=1}
          {interpMode=RCIM_Cubic, tangentMode=RCTM_User, time=0.7, value=1}
          {interpMode=RCIM_Cubic, tangentMode=RCTM_User, time=1, value=0.6}
         ]}}})
     Z=1
    ))
   &ScaleColor(
    Scale Alpha=&FloatFromCurve(FloatCurve=DataInterface() {propertyValues {Curve {keys [
         {interpMode=RCIM_Cubic, tangentMode=RCTM_User, time=0, value=0}
         {interpMode=RCIM_Cubic, tangentMode=RCTM_User, time=0.1, value=1}
         {interpMode=RCIM_Cubic, tangentMode=RCTM_User, time=0.85, value=1}
         {interpMode=RCIM_Cubic, tangentMode=RCTM_User, time=1, value=0}
        ]}}})
    ScaleA=true
    ScaleRGB=true
   )
   &DynamicMaterialParameters(
    Index 0 Param 1=&Add_Float(
     A=1
     B=&ReturnNormalizedExecIndex(Normalized Index Scale=&FloatFromCurve(FloatCurve=DataInterface() {propertyValues {Curve {keys [
           {arriveTangent=-0.0006, interpMode=RCIM_Cubic, leaveTangent=-0.0006, tangentMode=RCTM_User, time=0, value=3}
           {interpMode=RCIM_Cubic, tangentMode=RCTM_User, time=0.75, value=1}
          ]}}}))
    )
    Index 0 Param 2=-3
    Module.Param0WriteEnabled=true
    Module.Param1WriteEnabled=true
    Module.Param2WriteEnabled=true
    Module.Param3WriteEnabled=true
   )
  ]
  rendererClasses [&NiagaraMeshRendererProperties]
  renderers [{rendererClass=&NiagaraMeshRendererProperties, rendererIndex=0}]
 }
}
summary {
 systemName=NS_LWE_Sparks_Fire_Can_Blast_01_Big
 emitters [
  {emitterName=LWE_Base, rendererClasses []}
  {emitterName=LWE_EndFlames__Top, rendererClasses []}
  {emitterName=LWE_Glow, rendererClasses []}
  {emitterName=LWE_EndFlames_Bottom, rendererClasses []}
  {emitterName=Main_Blast, rendererClasses [&NiagaraMeshRendererProperties]}
  {emitterName=Light001, rendererClasses [&NiagaraLightRendererProperties]}
 ]
 userVariables []
}
